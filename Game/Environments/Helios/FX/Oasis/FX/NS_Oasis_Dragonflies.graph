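# Niagara system for ambient GPU dragonflies in the Helios Oasis, drawn with mesh and sprite renderers.
# Spawn rate follows user multipliers, time of day and distance culling, filling a flattened sphere.
# Vortex, curl noise, acceleration, gravity and camera repulsion drive flight, while a waveform-modulated speed limit varies pace and a kill volume bounds the swarm.
# A second emitter, a disabled variant, has higher spawn rate, stronger gravity and different speed limits.
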NiagaraSystem /Game/Environments/Helios/FX/Oasis/FX/NS_Oasis_Dragonflies
dependencies {
 usedDataInterfaces [&NiagaraDataInterfaceCurve, &NiagaraDataInterfaceCamera]
 usedDynamicInputs []
 usedModules [&SystemState]
}
emitters {
 Dragonfly_Flying {
  simTarget=GPUComputeSim
  emitterUpdateScript [
   &EmitterState(
    Enable Distance Culling=true
    Life Cycle Mode=::NewEnumerator1
    Loop Behavior=::NewEnumerator1
    Loop Duration Mode=::NewEnumerator1
    Min Distance Response=::SleepAndLetParticlesFinish
    MinDistance=250
    Module.bApplySpawnCountScaleByDistance=true
    Module.bUseMaxDistance=true
    Module.bUseMinDistance=true
    Scalability Mode=::NewEnumerator1
    Scale Spawn Count=true
    Spawn Count Scale By Distance=DataInterface() {propertyValues {Curve {keys [<at 0 1>,<at 0.95 1>,<at 1 0>]}}}
    UseLoopDelay=false
   )
   &SpawnRate(
    Spawn Probability=&FNTimeOfDayScalar(Night Scale=$User.Spawn at Night, Morning Scale=$User.Spawn at Morning, Day Scale=$User.Spawn at Day, Evening Scale=$User.Spawn at Evening)
    SpawnRate=&Multiply_Float(A=3, B=$User.Spawn Mult)
    Use Spawn Probability=true
   )
   /SetVariables(Emitter.CameraPosition:Vector3f=&ReturnCameraPosition(Coordinate Space:ENiagaraCoordinateSpace=::Simulation, Camera Query:NiagaraDataInterfaceCamera=DataInterface())) {isSetParameters=true, module=SetVariables_#1}
  ]
  particleSpawnScript [
   &InitializeParticle(
    AdjustAlpha=false
    AdjustHue=false
    AdjustSaturation=false
    AdjustValue=true
    Color Mode=::NewEnumerator3
    Lifetime=&RandomRangeFloat(Minimum=5, Maximum=8)
    Mesh Scale=&Multiply_Vector(
     A=&RandomRangeVector(Minimum=<1 1 1>, Maximum=<1.3 2 1>)
     B=&VectorFromFloat(Value=2)
    )
    Mesh Scale Mode=::NewEnumerator4
    Saturation Range:Vector2f=<0.8 233.6756>
    UsePositionOffset=false
    Value Range:Vector2f=<0.4 1.198>
    Write Material Random=true
   ) {module=InitializeParticle001}
   &SphereLocation(
    MaskSpawnGroup=false
    Non Uniform Scale=<1 1 0.1>
    Override Seed=false
    Sphere Radius=$User.Radius
    Surface Only=false
    UseNonUniformScale=true
    UseOffset=false
    UseSphereOrientation=false
   )
   &InitialMeshOrientation(
    Mesh Orientation Mode=::NewEnumerator1
    Rotation=&RandomRangeVector(Minimum=<0 0 0.75>, Maximum=<0 0 0.8>)
    Use Rotation Vector=true
   )
   &AddVelocity(Velocity=<0 0 -22>)
  ]
  particleUpdateScript [
   &ParticleState()
   /SetVariables(Particles.RandomDirection:Vector3f=&Lerp_Vector(
     A=&MakeVector(
      X=&Waveform(
       Global Phase=&Add_Float(A=$Particles.Age, B=&RandomRangeFloat(Minimum=0, Maximum=5))
       "Waveform [1]"=::NewEnumerator8
       "[1] Frequency"=1
       "[1] Amplitude Min/Max"=<-1 1>
       "[2] Phase"=$Emitter.Age
       "[3] Phase"=$Emitter.Age
      )
      Y=&Waveform(
       Global Phase=&Add_Float(A=$Particles.Age, B=&RandomRangeFloat(Minimum=0, Maximum=5))
       "Waveform [1]"=::NewEnumerator8
       "[1] Frequency"=1
       "[1] Amplitude Min/Max"=<-1 1>
       "[2] Phase"=$Emitter.Age
       "[3] Phase"=$Emitter.Age
      )
      Z=0
     )
     B=<1 0 0>
     Alpha=&Waveform(
      Global Phase=&Add_Float(A=$Particles.Age, B=&RandomRangeFloat(Minimum=0, Maximum=5))
      Global Clamp Min=0.4
      "Waveform [1]"=::NewEnumerator8
      "[1] Frequency"=1
      "[1] Amplitude Min/Max"=<0 1>
      "[2] Phase"=$Emitter.Age
      "[3] Phase"=$Emitter.Age
     )
    )) {enabled=false, isSetParameters=true, module=SetVariables_#2}
   &InterpolateOverTime(
    Target Vector Value:Vector3f=&NormalizeVector(Vector To Normalize=$Particles.RandomDirection, Safe Normalize=true, Safe Fallback Threshold=0, Safe Fallback Vector=<0 0 0.001>)
    Variable Type=::NewEnumerator2
   ) {enabled=false, module=InterpolateOverTime001}
   &WindForce(
    Air Resistance=0.1
    Wind Speed=&TransformVector(Vector=$Particles.InterpolateOverTime001.InterpolateOverTime_V3_.Moving Average, Destination Space=::Simulation)
    Wind Speed Scale=&RandomRangeFloat(Minimum=50, Maximum=100)
   ) {enabled=false}
   &PointAttractionForce(
    Attraction Radius=2500
    Attraction Strength=&RandomRangeFloat(Minimum=2, Maximum=3)
    Attractor Position:Vector3f=$Particles.FishMoverPosition
    Kill Within Radius=false
    Module.Use Falloff=true
    UseAttractorOffset=false
   ) {enabled=false}
   &PointAttractionForce(
    Attraction Radius=500
    Attraction Strength=&RandomRangeFloat(Minimum=0, Maximum=-600)
    Attractor Position:Vector3f=$Emitter.CameraPosition
    Kill Within Radius=false
    Module.Use Falloff=true
    UseAttractorOffset=false
   ) {module=PointAttractionForce002}
   &VortexForce(
    Origin Pull Amount=99
    UseOriginOffset=false
    Vortex Axis=&Lerp_Vector(
     A=<0.1 0.1467 1>
     B=<-0.5 0 0.25>
     Alpha=&FloatFromCurve(
      FloatCurve=DataInterface() {propertyValues {Curve {keys [
          {interpMode=RCIM_Cubic, tangentMode=RCTM_User, time=0, value=0}
          {interpMode=RCIM_Cubic, tangentMode=RCTM_User, time=0.5, value=1}
          {tangentMode=RCTM_User, time=1, value=0}
         ]}}}
      CurveIndex=$System.NormalizedLoopAge
     )
    )
    Vortex Force Amount=&RandomRangeFloat(Minimum=-10, Maximum=200)
   )
   &CurlNoiseForce(Noise Strength=938.7308, Pan Noise=false)
   &Drag(Use Linear Drag=true, Use Rotational Drag=true) {enabled=false}
   &AccelerationForce(Acceleration=&RandomRangeVector(Minimum=<35.3901 38.0236 0>, Maximum=<367.4797 395.9796 1>))
   &GravityForce(Gravity=&MakeVector(
     X=0
     Y=0
     Z=&FloatFromCurve(
      FloatCurve=DataInterface() {propertyValues {Curve {keys [
          {interpMode=RCIM_Cubic, tangentMode=RCTM_User, time=0, value=0}
          {arriveTangent=0.3998, interpMode=RCIM_Cubic, leaveTangent=0.3998, tangentMode=RCTM_User, time=1, value=1}
         ]}}}
      Scale Curve=-55
     )
    ))
   &SolveForcesAndVelocity(
    Clamp Velocity=true
    Limit Acceleration=false
    Speed Limit:NiagaraFloat=&Lerp_Float(
     A=&Multiply_Float(
      A=1
      B=&Waveform(
       "Waveform [1]"=::NewEnumerator0
       "[1] Phase Scale"=2
       "[1] Frequency"=&RandomRangeFloat(Minimum=1, Maximum=0.5)
       "[1] Amplitude Min/Max"=<0 1>
      )
     )
     B=&Multiply_Float(
      A=9999
      B=&Waveform(
       "Waveform [1]"=::NewEnumerator0
       "[1] Phase Scale"=&RandomRangeFloat(Minimum=0.3, Maximum=0.4)
       "[1] Frequency"=1
       "[1] Amplitude Min/Max"=<0 1>
      )
     )
     Alpha=&FloatFromCurve(FloatCurve=DataInterface() {propertyValues {Curve {keys [
          {interpMode=RCIM_Cubic, tangentMode=RCTM_User, time=0, value=1}
          <at 0.2 0.1 RCIM_Cubic>
          {interpMode=RCIM_Cubic, tangentMode=RCTM_User, time=0.5, value=1}
          <at 0.8 0.1 RCIM_Cubic>
          {tangentMode=RCTM_User, time=1, value=1}
         ]}}})
    )
   ) {module=SolveForcesAndVelocity002}
   &DynamicMaterialParameters(
    Index 0 Param 1=2
    Index 0 Param 2=&RandomRangeFloat(Minimum=2, Maximum=5)
    Index 0 Param 3=&RandomRangeFloat(Minimum=0.25, Maximum=0.5)
    Index 1 Param 1=0.35
    Module.Param0WriteEnabled=true
    Module.Param0WriteEnabled1=true
    Module.Param1WriteEnabled=true
    Module.Param1WriteEnabled1=true
    Module.Param2WriteEnabled=true
    Module.Param2WriteEnabled1=true
    Module.Param3WriteEnabled=true
    Module.Param3WriteEnabled1=true
    Write Parameter Index 1=true
   )
   &RegionMask(Outer Vector:Vector3f=<0 0 2>, Region Radius=2500, UseOffset=false, UseRotation=false, Value Type:ENiagaraRegionMaskValue=::NewEnumerator1)
   &KillParticlesInVolume(Invert Volume=true, Sphere Radius=2500)
   /SetVariables(Particles.New Vector 3f:Vector3f=&Multiply_VectorByFloat(
     Vector=$Particles.Scale
     Float=&FloatFromCurve(FloatCurve=DataInterface() {propertyValues {Curve {keys [
          {interpMode=RCIM_Cubic, tangentMode=RCTM_User, time=0, value=0}
          {interpMode=RCIM_Cubic, tangentMode=RCTM_User, time=0.1, value=1}
          {interpMode=RCIM_Cubic, tangentMode=RCTM_User, time=0.8, value=1}
          {tangentMode=RCTM_User, time=1, value=0}
         ]}}})
    )) {isSetParameters=true, module=SetVariables_#3}
   &ScaleColor(
    Scale Alpha=&Multiply_Float(A=&RandomRangeFloat(Minimum=0.6, Maximum=0.75), B=$User.Color Alpha)
    Scale RGB=&Multiply_VectorByFloat(Vector=<1 1 1>, Float=$User.Color Scale RGB)
    ScaleA=true
    ScaleRGB=true
   ) {module=ScaleColor001}
  ]
  rendererClasses [&NiagaraMeshRendererProperties, &NiagaraSpriteRendererProperties]
  renderers [
   {rendererClass=&NiagaraMeshRendererProperties, rendererIndex=0}
   {rendererClass=&NiagaraSpriteRendererProperties, rendererIndex=1}
   {rendererClass=&NiagaraMeshRendererProperties, rendererIndex=2}
  ]
 }
 Dragonfly_Flying003 {
  bEnabled=false
  simTarget=GPUComputeSim
  emitterUpdateScript [
   &EmitterState(
    Enable Distance Culling=true
    Life Cycle Mode=::NewEnumerator1
    Loop Behavior=::NewEnumerator1
    Loop Duration Mode=::NewEnumerator1
    Min Distance Response=::SleepAndLetParticlesFinish
    MinDistance=250
    Module.bApplySpawnCountScaleByDistance=true
    Module.bUseMaxDistance=true
    Module.bUseMinDistance=true
    Scalability Mode=::NewEnumerator1
    Scale Spawn Count=true
    Spawn Count Scale By Distance=DataInterface() {propertyValues {Curve {keys [<at 0 1>,<at 0.95 1>,<at 1 0>]}}}
    UseLoopDelay=false
   )
   &SpawnRate(
    Spawn Probability=&FNTimeOfDayScalar(Night Scale=$User.Spawn at Night, Morning Scale=$User.Spawn at Morning, Day Scale=$User.Spawn at Day, Evening Scale=$User.Spawn at Evening)
    SpawnRate=&Multiply_Float(A=6, B=$User.Spawn Mult)
    Use Spawn Probability=true
   )
   /SetVariables(Emitter.CameraPosition:Vector3f=&ReturnCameraPosition(Coordinate Space:ENiagaraCoordinateSpace=::Simulation, Camera Query:NiagaraDataInterfaceCamera=DataInterface())) {isSetParameters=true, module=SetVariables_#1}
  ]
  particleSpawnScript [
   &InitializeParticle(
    AdjustAlpha=false
    AdjustHue=false
    AdjustSaturation=false
    AdjustValue=true
    Color Mode=::NewEnumerator3
    Lifetime=&RandomRangeFloat(Minimum=5, Maximum=8)
    Mesh Scale=&Multiply_Vector(
     A=&RandomRangeVector(Minimum=<1 1 1>, Maximum=<1.3 2 1>)
     B=&VectorFromFloat(Value=2)
    )
    Mesh Scale Mode=::NewEnumerator4
    Saturation Range:Vector2f=<0.8 233.6756>
    UsePositionOffset=false
    Value Range:Vector2f=<0.4 1.198>
    Write Material Random=true
   ) {module=InitializeParticle001}
   &SphereLocation(
    MaskSpawnGroup=false
    Non Uniform Scale=<1 1 0.1>
    Override Seed=false
    Sphere Radius=$User.Radius
    Surface Only=false
    UseNonUniformScale=true
    UseOffset=false
    UseSphereOrientation=false
   )
   &InitialMeshOrientation(
    Mesh Orientation Mode=::NewEnumerator1
    Rotation=&RandomRangeVector(Minimum=<0 0 0.75>, Maximum=<0 0 0.8>)
    Use Rotation Vector=true
   )
   &AddVelocity(Velocity=<0 0 -22>)
  ]
  particleUpdateScript [
   &ParticleState()
   /SetVariables(Particles.RandomDirection:Vector3f=&Lerp_Vector(
     A=&MakeVector(
      X=&Waveform(
       Global Phase=&Add_Float(A=$Particles.Age, B=&RandomRangeFloat(Minimum=0, Maximum=5))
       "Waveform [1]"=::NewEnumerator8
       "[1] Frequency"=1
       "[1] Amplitude Min/Max"=<-1 1>
       "[2] Phase"=$Emitter.Age
       "[3] Phase"=$Emitter.Age
      )
      Y=&Waveform(
       Global Phase=&Add_Float(A=$Particles.Age, B=&RandomRangeFloat(Minimum=0, Maximum=5))
       "Waveform [1]"=::NewEnumerator8
       "[1] Frequency"=1
       "[1] Amplitude Min/Max"=<-1 1>
       "[2] Phase"=$Emitter.Age
       "[3] Phase"=$Emitter.Age
      )
      Z=0
     )
     B=<1 0 0>
     Alpha=&Waveform(
      Global Phase=&Add_Float(A=$Particles.Age, B=&RandomRangeFloat(Minimum=0, Maximum=5))
      Global Clamp Min=0.4
      "Waveform [1]"=::NewEnumerator8
      "[1] Frequency"=1
      "[1] Amplitude Min/Max"=<0 1>
      "[2] Phase"=$Emitter.Age
      "[3] Phase"=$Emitter.Age
     )
    )) {enabled=false, isSetParameters=true, module=SetVariables_#2}
   &InterpolateOverTime(
    Target Vector Value:Vector3f=&NormalizeVector(Vector To Normalize=$Particles.RandomDirection, Safe Normalize=true, Safe Fallback Threshold=0, Safe Fallback Vector=<0 0 0.001>)
    Variable Type=::NewEnumerator2
   ) {enabled=false, module=InterpolateOverTime001}
   &WindForce(
    Air Resistance=0.1
    Wind Speed=&TransformVector(Vector=$Particles.InterpolateOverTime001.InterpolateOverTime_V3_.Moving Average, Destination Space=::Simulation)
    Wind Speed Scale=&RandomRangeFloat(Minimum=50, Maximum=100)
   ) {enabled=false}
   &PointAttractionForce(
    Attraction Radius=2500
    Attraction Strength=&RandomRangeFloat(Minimum=2, Maximum=3)
    Attractor Position:Vector3f=$Particles.FishMoverPosition
    Kill Within Radius=false
    Module.Use Falloff=true
    UseAttractorOffset=false
   ) {enabled=false}
   &PointAttractionForce(
    Attraction Radius=500
    Attraction Strength=&RandomRangeFloat(Minimum=0, Maximum=-600)
    Attractor Position:Vector3f=$Emitter.CameraPosition
    Kill Within Radius=false
    Module.Use Falloff=true
    UseAttractorOffset=false
   ) {module=PointAttractionForce002}
   &VortexForce(
    Origin Pull Amount=99
    UseOriginOffset=false
    Vortex Axis=&Lerp_Vector(
     A=<0.1 0.1467 1>
     B=<-0.5 0 0.25>
     Alpha=&FloatFromCurve(
      FloatCurve=DataInterface() {propertyValues {Curve {keys [
          {interpMode=RCIM_Cubic, tangentMode=RCTM_User, time=0, value=0}
          {interpMode=RCIM_Cubic, tangentMode=RCTM_User, time=0.5, value=1}
          {tangentMode=RCTM_User, time=1, value=0}
         ]}}}
      CurveIndex=$System.NormalizedLoopAge
     )
    )
    Vortex Force Amount=&RandomRangeFloat(Minimum=-10, Maximum=200)
   )
   &CurlNoiseForce(Noise Strength=938.7308, Pan Noise=false)
   &Drag(Use Linear Drag=true, Use Rotational Drag=true) {enabled=false}
   &AccelerationForce(Acceleration=&RandomRangeVector(Minimum=<35.3901 38.0236 0>, Maximum=<367.4797 395.9796 1>))
   &GravityForce(Gravity=&MakeVector(
     X=0
     Y=0
     Z=&FloatFromCurve(
      FloatCurve=DataInterface() {propertyValues {Curve {keys [
          {interpMode=RCIM_Cubic, tangentMode=RCTM_User, time=0, value=0}
          {arriveTangent=0.3998, interpMode=RCIM_Cubic, leaveTangent=0.3998, tangentMode=RCTM_User, time=1, value=1}
         ]}}}
      Scale Curve=-222
     )
    ))
   &SolveForcesAndVelocity(
    Clamp Velocity=true
    Limit Acceleration=false
    Speed Limit:NiagaraFloat=&Lerp_Float(
     A=&Multiply_Float(
      A=3333
      B=&Waveform(
       "Waveform [1]"=::NewEnumerator0
       "[1] Frequency"=&RandomRangeFloat(Minimum=0.1, Maximum=0.5)
       "[1] Amplitude Min/Max"=<0 1>
      )
     )
     B=&Multiply_Float(
      A=2222
      B=&Waveform(
       "Waveform [1]"=::NewEnumerator1
       "[1] Frequency"=&RandomRangeFloat(Minimum=0.1, Maximum=0.5)
       "[1] Amplitude Min/Max"=<0 0.5>
      )
     )
     Alpha=&MakeFloatFromInt(Int=&RandomRangeInt(Minimum=0, Maximum=1))
    )
   ) {enabled=false}
   &SolveForcesAndVelocity(
    Clamp Velocity=true
    Limit Acceleration=false
    Speed Limit:NiagaraFloat=&Lerp_Float(
     A=&Multiply_Float(
      A=3333
      B=&Waveform(
       "Waveform [1]"=::NewEnumerator0
       "[1] Frequency"=&RandomRangeFloat(Minimum=0.1, Maximum=0.5)
       "[1] Amplitude Min/Max"=<0 1>
      )
     )
     B=&Multiply_Float(
      A=3333
      B=&Waveform(
       "Waveform [1]"=::NewEnumerator1
       "[1] Frequency"=&RandomRangeFloat(Minimum=0.1, Maximum=0.5)
       "[1] Amplitude Min/Max"=<0.1 1>
      )
     )
     Alpha=&FloatFromCurve(FloatCurve=DataInterface() {propertyValues {Curve {keys [
          {interpMode=RCIM_Cubic, tangentMode=RCTM_User, time=0, value=0}
          {arriveTangent=4.3451, interpMode=RCIM_Cubic, leaveTangent=-4.1667, tangentMode=RCTM_Break, time=0.5, value=1}
          {tangentMode=RCTM_User, time=1, value=0}
         ]}}})
    )
   ) {enabled=false, module=SolveForcesAndVelocity001}
   &SolveForcesAndVelocity(
    Clamp Velocity=true
    Limit Acceleration=false
    Speed Limit:NiagaraFloat=&Lerp_Float(
     A=&Multiply_Float(
      A=3333
      B=&Waveform(
       "Waveform [1]"=::NewEnumerator0
       "[1] Frequency"=&RandomRangeFloat(Minimum=0.1, Maximum=0.5)
       "[1] Amplitude Min/Max"=<0 1>
      )
     )
     B=2222
     Alpha=&FloatFromCurve(FloatCurve=DataInterface() {propertyValues {Curve {keys [
          {interpMode=RCIM_Cubic, tangentMode=RCTM_User, time=0, value=0}
          {interpMode=RCIM_Cubic, tangentMode=RCTM_User, time=0.5, value=1}
          {tangentMode=RCTM_User, time=1, value=0}
         ]}}})
    )
   ) {module=SolveForcesAndVelocity002}
   &DynamicMaterialParameters(
    Index 0 Param 1=2
    Index 0 Param 2=&RandomRangeFloat(Minimum=2, Maximum=5)
    Index 0 Param 3=&RandomRangeFloat(Minimum=0.25, Maximum=0.5)
    Index 1 Param 1=0.35
    Module.Param0WriteEnabled=true
    Module.Param0WriteEnabled1=true
    Module.Param1WriteEnabled=true
    Module.Param1WriteEnabled1=true
    Module.Param2WriteEnabled=true
    Module.Param2WriteEnabled1=true
    Module.Param3WriteEnabled=true
    Module.Param3WriteEnabled1=true
    Write Parameter Index 1=true
   )
   &RegionMask(Outer Vector:Vector3f=<0 0 2>, Region Radius=2500, UseOffset=false, UseRotation=false, Value Type:ENiagaraRegionMaskValue=::NewEnumerator1)
   &KillParticlesInVolume(Invert Volume=true, Sphere Radius=2500)
   /SetVariables(Particles.New Vector 3f:Vector3f=&Multiply_VectorByFloat(
     Vector=$Particles.Scale
     Float=&FloatFromCurve(FloatCurve=DataInterface() {propertyValues {Curve {keys [
          {interpMode=RCIM_Cubic, tangentMode=RCTM_User, time=0, value=0}
          {interpMode=RCIM_Cubic, tangentMode=RCTM_User, time=0.1, value=1}
          {interpMode=RCIM_Cubic, tangentMode=RCTM_User, time=0.8, value=1}
          {tangentMode=RCTM_User, time=1, value=0}
         ]}}})
    )) {isSetParameters=true, module=SetVariables_#3}
   &ScaleColor(
    Scale Alpha=&Multiply_Float(A=&RandomRangeFloat(Minimum=0.6, Maximum=0.75), B=$User.Color Alpha)
    Scale RGB=&Multiply_VectorByFloat(Vector=<1 1 1>, Float=$User.Color Scale RGB)
    ScaleA=true
    ScaleRGB=true
   ) {module=ScaleColor001}
  ]
  rendererClasses [&NiagaraMeshRendererProperties, &NiagaraSpriteRendererProperties]
  renderers [
   {rendererClass=&NiagaraMeshRendererProperties, rendererIndex=0}
   {rendererClass=&NiagaraSpriteRendererProperties, rendererIndex=1}
   {rendererClass=&NiagaraMeshRendererProperties, rendererIndex=2}
  ]
 }
}
summary {
 systemName=NS_Oasis_Dragonflies
 emitters [
  {
   bEnabled=false
   emitterName=Dragonfly_Flying003
   simTarget=GPUComputeSim
   rendererClasses [&NiagaraMeshRendererProperties, &NiagaraSpriteRendererProperties]
  }
  {
   emitterName=Dragonfly_Flying
   simTarget=GPUComputeSim
   rendererClasses [&NiagaraMeshRendererProperties, &NiagaraSpriteRendererProperties]
  }
 ]
 userVariables [
  User.bIsInWater {type=NiagaraBool}
  User.Color Alpha {type=NiagaraFloat, defaultValue=1, description=""}
  User.Color Scale RGB {type=NiagaraFloat, defaultValue=1, description=""}
  User.Radius {type=NiagaraFloat, defaultValue=500, description=""}
  User.Scale Mult {type=NiagaraFloat, defaultValue=1, description=""}
  User.Spawn at Day {type=NiagaraFloat, defaultValue=1, description=""}
  User.Spawn at Evening {type=NiagaraFloat, defaultValue=1, description=""}
  User.Spawn at Morning {type=NiagaraFloat, defaultValue=1, description=""}
  User.Spawn at Night {type=NiagaraFloat, defaultValue=1, description=""}
  User.Spawn Mult {type=NiagaraFloat, defaultValue=1, description=""}
  User.SpawnRate {type=NiagaraFloat, defaultValue=1, description=""}
  User.SphereRadius {type=NiagaraFloat, defaultValue=500, description=""}
 ]
}
