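# Niagara system for the Helios Oasis environment that renders a flock of circling vultures as GPU mesh particles.
# Only Vultures_Circling001 is active (the other emitter is disabled), spawning randomly scaled birds on a torus, with distance culling and spawn scaling.
# A vortex force with origin pull drives the circling, while curves control late-life drag, mesh rotation and grow/shrink scaling.
# Dynamic material parameters feed randomized, curve-timed values to the mesh material.
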NiagaraSystem /Game/Environments/Helios/FX/Oasis/FX/NS_Oasis_Vulture_Circling
dependencies {
 usedDataInterfaces [&NiagaraDataInterfaceCamera, &NiagaraDataInterfaceCurve]
 usedDynamicInputs [&SimulationPosition]
 usedModules [&SystemState]
}
emitters {
 Vultures_Circling {
  bEnabled=false
  simTarget=GPUComputeSim
  emitterUpdateScript [
   /SetVariables(Emitter.CameraPosition:Vector3f=&ReturnCameraVector(Camera Property=::NewEnumerator0, Coordinate Space=::Simulation)) {enabled=false, isSetParameters=true}
   &EmitterState(
    Enable Distance Culling=true
    Life Cycle Mode=::NewEnumerator1
    Loop Behavior=::NewEnumerator1
    Loop Duration Mode=::NewEnumerator1
    MaxDistance=35000
    Min Distance Response=::SleepAndLetParticlesFinish
    MinDistance=2500
    Module.bApplySpawnCountScaleByDistance=true
    Module.bUseMaxDistance=true
    Module.bUseMinDistance=true
    Scalability Mode=::NewEnumerator1
    Scale Spawn Count=true
    Spawn Count Scale By Distance=DataInterface() {propertyValues {Curve {keys [<at 0 1>,<at 0.95 1>,<at 1 0>]}}}
    UseLoopDelay=false
   )
   &SpawnRate(SpawnRate=2, Use Spawn Probability=false) {module=SpawnRate001}
  ]
  particleSpawnScript [
   &InitializeParticle(
    Lifetime Max=8
    Lifetime Min=5
    Lifetime Mode=::NewEnumerator1
    Mass Mode=::NewEnumerator1
    Mesh Scale=&Lerp_Vector(A=<1.5 1.5 1.5>, B=<1.875 1.875 1.875>, Alpha=&UniformAOrBFloat(A=0))
    Mesh Scale Mode=::NewEnumerator4
    Sprite Size Mode=::NewEnumerator4
    UsePositionOffset=false
    Write Material Random=true
   ) {enabled=false, module=InitializeParticle002}
   &InitializeParticle(
    Lifetime Max=8
    Lifetime Min=5
    Lifetime Mode=::NewEnumerator1
    Mass Mode=::NewEnumerator1
    Mesh Scale=&Lerp_Vector(
     A=<1.5 1.5 1.5>
     B=<1.875 1.875 1.875>
     Alpha=&UniformAOrBFloat(A=0, Evaluation Type=::NewEnumerator0, Recalculate Random Each Loop=true, Override Seed=false)
    )
    Mesh Scale Mode=::NewEnumerator4
    Sprite Size Mode=::NewEnumerator4
    UsePositionOffset=false
    Write Material Random=true
   )
   &TorusLocation(
    Align Torus With Axis=false
    Enable Torus Knot Radius=false
    Handle Radius=1000
    Large Radius:NiagaraFloat=&RandomRangeFloat(Minimum=1500, Maximum=3000)
    MaskSpawnGroup=false
    Module.Override Randomness=false
    Torus Mode=::NewEnumerator1
   )
  ]
  particleUpdateScript [
   &ParticleState()
   &VortexForce(
    Origin Pull Amount:NiagaraFloat=&RandomRangeFloat(Minimum=50, Maximum=120)
    Origin Pull Only On Vortex Axis=false
    UseOriginOffset=true
    Vortex Axis=<0 0 4>
    Vortex Force Amount=100
    Vortex Origin Offset:Vector3f=&RandomRangeVector(Minimum=<0 0 222>, Maximum=<0 0 -222>)
   )
   &Drag(
    Drag=&FloatFromCurve(
     FloatCurve=DataInterface() {propertyValues {Curve {keys [
         {interpMode=RCIM_Cubic, tangentMode=RCTM_User, time=0, value=0}
         {arriveTangent=0.4772, interpMode=RCIM_Cubic, leaveTangent=0.4772, tangentMode=RCTM_User, time=0.8, value=0}
         {arriveTangent=2.2717, interpMode=RCIM_Cubic, leaveTangent=2.2717, tangentMode=RCTM_User, time=1, value=1}
        ]}}}
     Scale Curve=3
    )
    Use Linear Drag=true
    Use Rotational Drag=true
   )
   &SolveForcesAndVelocity(Clamp Velocity=false, Limit Acceleration=false)
   &UpdateMeshOrientation(Rotation Rate=&FloatFromCurve(
     FloatCurve=DataInterface() {propertyValues {Curve {keys [
         {arriveTangent=-0.5868, interpMode=RCIM_Cubic, leaveTangent=-0.5868, tangentMode=RCTM_User, time=0, value=0}
         {arriveTangent=0.25, interpMode=RCIM_Cubic, leaveTangent=0.25, time=0.3, value=0}
         <at 0.4 0.1 RCIM_Cubic>
         {arriveTangent=-0.1667, interpMode=RCIM_Cubic, leaveTangent=-0.1667, time=0.6, value=0}
         {interpMode=RCIM_Cubic, tangentMode=RCTM_User, time=1, value=0}
        ]}}}
     Scale Curve=&Multiply_FloatByInt(
      Float=&RandomRangeFloat(Minimum=-1, Maximum=-3)
      Integer=&RandomRangeInt(Minimum=1, Maximum=1)
     )
    ))
   &ScaleMeshSize(Scale Factor=&VectorFromFloat(Value=&FloatFromCurve(FloatCurve=DataInterface() {propertyValues {Curve {keys [
          {interpMode=RCIM_Cubic, tangentMode=RCTM_User, time=0, value=0}
          {interpMode=RCIM_Cubic, tangentMode=RCTM_Break, time=0.2, value=1}
          {arriveTangent=-0.0529, interpMode=RCIM_Cubic, leaveTangent=-1.25, tangentMode=RCTM_Break, time=0.8, value=1}
          {tangentMode=RCTM_User, time=1, value=0}
         ]}}})))
   &Color()
   &DynamicMaterialParameters(
    Index 0 Param 1=0.25
    Index 0 Param 2=&Multiply_Float(
     A=0.4
     B=&FloatFromCurve(
      FloatCurve=DataInterface() {propertyValues {Curve {keys [
          {interpMode=RCIM_Cubic, tangentMode=RCTM_User, time=0, value=0}
          <at 0.1 0 RCIM_Cubic>
          {arriveTangent=0.0316, interpMode=RCIM_Cubic, leaveTangent=4, tangentMode=RCTM_Break, time=0.55, value=0}
          <at 0.65 0.7 RCIM_Cubic>
          {arriveTangent=-4, interpMode=RCIM_Cubic, leaveTangent=-0.0315, tangentMode=RCTM_Break, time=0.75, value=0}
          {tangentMode=RCTM_User, time=1, value=0}
         ]}}}
      Scale Curve=&Multiply_FloatByInt(Float=0.0003, Integer=&RandomRangeInt(Minimum=0, Maximum=1))
     )
    )
    Index 0 Param 3=0.17
    Index 1 Param 1=1
    Index 1 Param 2=&RandomRangeFloat(Minimum=8, Maximum=10)
    Index 1 Param 3=2
    Index 2 Param 2=0.4
    Index 3 Param 1=1
    Index 3 Param 4=1
    Module.Param0WriteEnabled=true
    Module.Param0WriteEnabled1=true
    Module.Param0WriteEnabled2=true
    Module.Param0WriteEnabled3=true
    Module.Param1WriteEnabled=true
    Module.Param1WriteEnabled1=true
    Module.Param1WriteEnabled2=true
    Module.Param1WriteEnabled3=true
    Module.Param2WriteEnabled=true
    Module.Param2WriteEnabled1=true
    Module.Param2WriteEnabled2=true
    Module.Param2WriteEnabled3=true
    Module.Param3WriteEnabled=true
    Module.Param3WriteEnabled1=true
    Module.Param3WriteEnabled2=true
    Module.Param3WriteEnabled3=true
    Write Parameter Index 1=true
    Write Parameter Index 2=true
    Write Parameter Index 3=true
   ) {enabled=false, module=DynamicMaterialParameters001}
   &DynamicMaterialParameters(
    Index 0 Param 1=0.25
    Index 0 Param 2=&RemapRange(
     Input Value To Remap=&FloatFromCurve(
      FloatCurve=DataInterface() {propertyValues {Curve {keys [
          {interpMode=RCIM_Cubic, tangentMode=RCTM_User, time=0, value=0}
          <at 0.1 0 RCIM_Cubic>
          {arriveTangent=0.0316, interpMode=RCIM_Cubic, leaveTangent=4, tangentMode=RCTM_Break, time=0.55, value=0}
          <at 0.65 0.7 RCIM_Cubic>
          {arriveTangent=-4, interpMode=RCIM_Cubic, leaveTangent=-0.0315, tangentMode=RCTM_Break, time=0.75, value=0}
          {tangentMode=RCTM_User, time=1, value=0}
         ]}}}
      Scale Curve=&Multiply_FloatByInt(Float=1, Integer=&RandomRangeInt(Minimum=1, Maximum=1))
     )
     Input Min=0
     Input Max=1
     Output Min=0
     Output Max=0.001
    )
    Index 0 Param 3=0.17
    Index 1 Param 1=1
    Index 1 Param 2=&RandomRangeFloat(Minimum=8, Maximum=10)
    Index 1 Param 3=2
    Index 2 Param 2=0.4
    Index 3 Param 1=1
    Index 3 Param 4=1
    Module.Param0WriteEnabled=true
    Module.Param0WriteEnabled1=true
    Module.Param0WriteEnabled2=true
    Module.Param0WriteEnabled3=true
    Module.Param1WriteEnabled=true
    Module.Param1WriteEnabled1=true
    Module.Param1WriteEnabled2=true
    Module.Param1WriteEnabled3=true
    Module.Param2WriteEnabled=true
    Module.Param2WriteEnabled1=true
    Module.Param2WriteEnabled2=true
    Module.Param2WriteEnabled3=true
    Module.Param3WriteEnabled=true
    Module.Param3WriteEnabled1=true
    Module.Param3WriteEnabled2=true
    Module.Param3WriteEnabled3=true
    Write Parameter Index 1=true
    Write Parameter Index 2=true
    Write Parameter Index 3=true
   ) {enabled=false}
   &DynamicMaterialParameters(
    Index 0 Param 1=0.25
    Index 0 Param 2=&FloatFromCurve(
     FloatCurve=DataInterface() {propertyValues {Curve {keys [
         {interpMode=RCIM_Cubic, tangentMode=RCTM_User, time=0, value=0}
         <at 0.1 0 RCIM_Cubic>
         {arriveTangent=0.0316, interpMode=RCIM_Cubic, leaveTangent=4, tangentMode=RCTM_Break, time=0.55, value=0}
         <at 0.65 0.7 RCIM_Cubic>
         {arriveTangent=-4, interpMode=RCIM_Cubic, leaveTangent=-0.0315, tangentMode=RCTM_Break, time=0.75, value=0}
         {tangentMode=RCTM_User, time=1, value=0}
        ]}}}
     Scale Curve=&Multiply_FloatByInt(Float=1, Integer=&RandomRangeInt(Minimum=0, Maximum=1))
    )
    Index 0 Param 3=0.17
    Index 1 Param 1=1
    Index 1 Param 2=&RandomRangeFloat(Minimum=8, Maximum=10)
    Index 1 Param 3=2
    Index 2 Param 2=0.4
    Index 3 Param 1=1
    Index 3 Param 4=1
    Module.Param0WriteEnabled=true
    Module.Param0WriteEnabled1=true
    Module.Param0WriteEnabled2=true
    Module.Param0WriteEnabled3=true
    Module.Param1WriteEnabled=true
    Module.Param1WriteEnabled1=true
    Module.Param1WriteEnabled2=true
    Module.Param1WriteEnabled3=true
    Module.Param2WriteEnabled=true
    Module.Param2WriteEnabled1=true
    Module.Param2WriteEnabled2=true
    Module.Param2WriteEnabled3=true
    Module.Param3WriteEnabled=true
    Module.Param3WriteEnabled1=true
    Module.Param3WriteEnabled2=true
    Module.Param3WriteEnabled3=true
    Write Parameter Index 1=true
    Write Parameter Index 2=true
    Write Parameter Index 3=true
   ) {module=DynamicMaterialParameters002}
  ]
  rendererClasses [&NiagaraMeshRendererProperties]
  renderers [{rendererClass=&NiagaraMeshRendererProperties, rendererIndex=0}]
 }
 Vultures_Circling001 {
  simTarget=GPUComputeSim
  emitterUpdateScript [
   &EmitterState(
    Enable Distance Culling=true
    Life Cycle Mode=::NewEnumerator1
    Loop Behavior=::NewEnumerator1
    Loop Duration Mode=::NewEnumerator1
    MaxDistance=35000
    Min Distance Response=::SleepAndLetParticlesFinish
    MinDistance=2500
    Module.bApplySpawnCountScaleByDistance=true
    Module.bUseMaxDistance=true
    Module.bUseMinDistance=true
    Scalability Mode=::NewEnumerator1
    Scale Spawn Count=true
    Spawn Count Scale By Distance=DataInterface() {propertyValues {Curve {keys [<at 0 1>,<at 0.95 1>,<at 1 0>]}}}
    UseLoopDelay=false
   )
   &SpawnRate(SpawnRate=2, Use Spawn Probability=false) {module=SpawnRate001}
  ]
  particleSpawnScript [
   &InitializeParticle(
    Lifetime Max=8
    Lifetime Min=5
    Lifetime Mode=::NewEnumerator1
    Mass Mode=::NewEnumerator1
    Mesh Scale=&Lerp_Vector(
     A=<1.5 1.5 1.5>
     B=<1.875 1.875 1.875>
     Alpha=&UniformAOrBFloat(A=0, Evaluation Type=::NewEnumerator0, Recalculate Random Each Loop=true, Override Seed=false)
    )
    Mesh Scale Mode=::NewEnumerator4
    Sprite Size Mode=::NewEnumerator4
    UsePositionOffset=false
    Write Material Random=true
   )
   &TorusLocation(
    Align Torus With Axis=false
    Enable Torus Knot Radius=false
    Handle Radius=1000
    Large Radius:NiagaraFloat=&RandomRangeFloat(Minimum=1500, Maximum=3000)
    MaskSpawnGroup=false
    Module.Override Randomness=false
    Torus Mode=::NewEnumerator1
   )
  ]
  particleUpdateScript [
   &ParticleState()
   &VortexForce(
    Origin Pull Amount:NiagaraFloat=&RandomRangeFloat(Minimum=50, Maximum=120)
    Origin Pull Only On Vortex Axis=false
    UseOriginOffset=true
    Vortex Axis=<0 0 4>
    Vortex Force Amount=100
    Vortex Origin Offset:Vector3f=&RandomRangeVector(Minimum=<0 0 222>, Maximum=<0 0 -222>)
   )
   &Drag(
    Drag=&FloatFromCurve(
     FloatCurve=DataInterface() {propertyValues {Curve {keys [
         {interpMode=RCIM_Cubic, tangentMode=RCTM_User, time=0, value=0}
         {arriveTangent=0.4772, interpMode=RCIM_Cubic, leaveTangent=0.4772, tangentMode=RCTM_User, time=0.8, value=0}
         {arriveTangent=2.2717, interpMode=RCIM_Cubic, leaveTangent=2.2717, tangentMode=RCTM_User, time=1, value=1}
        ]}}}
     Scale Curve=3
    )
    Use Linear Drag=true
    Use Rotational Drag=true
   )
   &SolveForcesAndVelocity(Clamp Velocity=false, Limit Acceleration=false)
   &UpdateMeshOrientation(Rotation Rate=&FloatFromCurve(
     FloatCurve=DataInterface() {propertyValues {Curve {keys [
         {arriveTangent=-0.5868, interpMode=RCIM_Cubic, leaveTangent=-0.5868, tangentMode=RCTM_User, time=0, value=0}
         {arriveTangent=0.25, interpMode=RCIM_Cubic, leaveTangent=0.25, time=0.3, value=0}
         <at 0.4 0.1 RCIM_Cubic>
         {arriveTangent=-0.1667, interpMode=RCIM_Cubic, leaveTangent=-0.1667, time=0.6, value=0}
         {interpMode=RCIM_Cubic, tangentMode=RCTM_User, time=1, value=0}
        ]}}}
     Scale Curve=&Multiply_FloatByInt(
      Float=&RandomRangeFloat(Minimum=-1, Maximum=-3)
      Integer=&RandomRangeInt(Minimum=1, Maximum=1)
     )
    ))
   &ScaleMeshSize(Scale Factor=&VectorFromFloat(Value=&FloatFromCurve(FloatCurve=DataInterface() {propertyValues {Curve {keys [
          {interpMode=RCIM_Cubic, tangentMode=RCTM_User, time=0, value=0}
          {interpMode=RCIM_Cubic, tangentMode=RCTM_Break, time=0.2, value=1}
          {arriveTangent=-0.0529, interpMode=RCIM_Cubic, leaveTangent=-1.25, tangentMode=RCTM_Break, time=0.8, value=1}
          {tangentMode=RCTM_User, time=1, value=0}
         ]}}})))
   &Color()
   &DynamicMaterialParameters(
    Index 0 Param 1=0.25
    Index 0 Param 2=&FloatFromCurve(
     FloatCurve=DataInterface() {propertyValues {Curve {keys [
         {interpMode=RCIM_Cubic, tangentMode=RCTM_User, time=0, value=0}
         <at 0.1 0 RCIM_Cubic>
         {arriveTangent=0.0316, interpMode=RCIM_Cubic, leaveTangent=4, tangentMode=RCTM_Break, time=0.55, value=0}
         <at 0.65 0.7 RCIM_Cubic>
         {arriveTangent=-4, interpMode=RCIM_Cubic, leaveTangent=-0.0315, tangentMode=RCTM_Break, time=0.75, value=0}
         {tangentMode=RCTM_User, time=1, value=0}
        ]}}}
     Scale Curve=&Multiply_FloatByInt(Float=1, Integer=&RandomRangeInt(Minimum=0, Maximum=1))
    )
    Index 0 Param 3=0.17
    Index 1 Param 1=1
    Index 1 Param 2=&RandomRangeFloat(Minimum=8, Maximum=10)
    Index 1 Param 3=2
    Index 2 Param 2=0.4
    Index 3 Param 1=1
    Index 3 Param 4=1
    Module.Param0WriteEnabled=true
    Module.Param0WriteEnabled1=true
    Module.Param0WriteEnabled2=true
    Module.Param0WriteEnabled3=true
    Module.Param1WriteEnabled=true
    Module.Param1WriteEnabled1=true
    Module.Param1WriteEnabled2=true
    Module.Param1WriteEnabled3=true
    Module.Param2WriteEnabled=true
    Module.Param2WriteEnabled1=true
    Module.Param2WriteEnabled2=true
    Module.Param2WriteEnabled3=true
    Module.Param3WriteEnabled=true
    Module.Param3WriteEnabled1=true
    Module.Param3WriteEnabled2=true
    Module.Param3WriteEnabled3=true
    Write Parameter Index 1=true
    Write Parameter Index 2=true
    Write Parameter Index 3=true
   ) {module=DynamicMaterialParameters002}
  ]
  rendererClasses [&NiagaraMeshRendererProperties]
  renderers [{rendererClass=&NiagaraMeshRendererProperties, rendererIndex=0}]
 }
}
summary {
 systemName=NS_Oasis_Vulture_Circling
 emitters [
  {bEnabled=false, emitterName=Vultures_Circling, simTarget=GPUComputeSim, rendererClasses [&NiagaraMeshRendererProperties]}
  {emitterName=Vultures_Circling001, simTarget=GPUComputeSim, rendererClasses [&NiagaraMeshRendererProperties]}
 ]
 userVariables []
}
