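# Unlit translucent two-sided material for Oasis dragonfly Niagara mesh particles, with shader-driven wing flapping.
# The dragonfly texture times particle color feeds an emissive-adjustment function.
# Sine-driven rotations about the particle pivot swing two wing pairs oppositely, masked by vertex color R and G, with extra UV channels selecting left or right vectors.
# A dynamic parameter sets amplitude, period, curvature and offset, with UV0 varying bend across the wing.
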Material /Game/Environments/Helios/FX/Oasis/Materials/M_Oasis_Dragonflies
graph {
 properties {
  bAutomaticallySetUsageInEditor=false
  bUsedWithNiagaraMeshParticles=true
  blendMode=BLEND_Translucent
  pixelDepthOffsetMode=PDOM_Legacy
  shadingModel=MSM_Unlit
  twoSided=true
  usedShadingModels=Unlit
 }
 nodes [
  MaterialFunctionCall(Emissive=Multiply(A=TextureSample() {texture=/Game/Environments/Helios/FX/Oasis/Textures/T_Dragonfly}.RGB, B=ParticleColor().RGB)) {materialFunction=&MF_FX_EmissiveAdjustments}
  n1 = Floor(None=Add(A=TextureCoordinate() {coordinateIndex=1}, B=Constant() {r=0.5}))
  NamedRerouteDeclaration(None=Clamp(None=ComponentMask(None=@n1) {g=true}) {desc=Wings L R}) {name=LeftRightSelect, nodeColor=<rgba 0.7344 0 1 1>}
  n2 = Floor(None=Add(A=TextureCoordinate() {coordinateIndex=2}, B=Constant() {r=0.5}))
  NamedRerouteDeclaration(None=Clamp(None=ComponentMask(None=@n2) {g=true}) {desc=Wings L R}) {name=LeftRightBackSelect, nodeColor=<rgba 1 0.6328 0 1>}
  n3 = Multiply(
   A=VertexColor().R
   B=LinearInterpolate(
    A=VectorParameter() {parameterName=WingLeft, defaultValue=<rgba 0 -0.25 -0.5 0>}.RGB
    B=VectorParameter() {parameterName=WingRight, defaultValue=<rgba 0 -0.25 0.5 0>}.RGB
    Alpha=NamedRerouteUsage()
   )
  )
  n4 = Constant3Vector() {constant=<rgba 0 1 0 0>}
  n5 = DynamicParameter() {defaultValue=<rgba 1.5 4 -0.04 -0.2>, paramNames [Amplitude,Period,Curvature,Offset]}
  n6 = Multiply(A=Time(), B=@n5.Period)
  n7 = Multiply(A=Constant() {r=0.08}, B=Add(A=Sine(None=@n6), B=@n5.Offset) {constB=-0.35})
  n8 = Add(A=ComponentMask(None=TextureCoordinate()) {r=true}, B=@n5.Curvature) {constB=0.5}
  n9 = RotateAboutAxis(
   NormalizedRotationAxis=Normalize(VectorInput=Transform(None=@n4) {transformSourceType=TRANSFORMSOURCE_ParticleWorld})
   RotationAngle=Multiply(A=Multiply(A=@n7, B=@n8), B=@n5.Amplitude)
   PivotPoint=MaterialFunctionCall() {materialFunction=&ObjectPivotPoint}.Mesh Particle Pivot Location
   Position=WorldPosition().XYZ
  )
  n10 = Multiply(
   A=VertexColor().G
   B=LinearInterpolate(
    A=VectorParameter() {parameterName=WingLeft, defaultValue=<rgba 0 -0.25 -0.5 0>}.RGB
    B=VectorParameter() {parameterName=WingRight, defaultValue=<rgba 0 -0.25 0.5 0>}.RGB
    Alpha=NamedRerouteUsage()
   )
  )
  n11 = Constant3Vector() {constant=<rgba 0 1 0 0>}
  n12 = Multiply(A=Time(), B=@n5.Period) {constB=5}
  n13 = Multiply(A=Constant() {r=-0.08}, B=Add(A=Sine(None=@n12), B=@n5.Offset) {constB=-0.35})
  n14 = Add(A=ComponentMask(None=TextureCoordinate()) {r=true}, B=@n5.Curvature) {constB=0.5}
  n15 = RotateAboutAxis(
   NormalizedRotationAxis=Normalize(VectorInput=Transform(None=@n11) {transformSourceType=TRANSFORMSOURCE_ParticleWorld})
   RotationAngle=Multiply(A=@n5.Amplitude, B=Multiply(A=@n13, B=@n14))
   PivotPoint=MaterialFunctionCall() {materialFunction=&ObjectPivotPoint}.Mesh Particle Pivot Location
   Position=WorldPosition().XYZ
  )
  Reroute(None=Add(A=Multiply(A=@n3, B=@n9), B=Multiply(A=@n10, B=@n15)))
 ]
}
